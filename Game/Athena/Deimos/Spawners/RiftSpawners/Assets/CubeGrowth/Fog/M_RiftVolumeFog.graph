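# Additive volume-domain material for rift spawner fog, shaping a tinted, glowing spherical puff of volumetric fog.
# A sphere mask around the object position plus an offset has scaled, animated noise subtracted to erode it, then drives the color and emissive terms.
# The noise takes the minimum of two scrolling world-space planar samples (XY, YZ), each shifted by the third axis to vary through depth.
# A quality switch replaces the noise with a constant on Low and Medium.
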Material /Game/Athena/Deimos/Spawners/RiftSpawners/Assets/CubeGrowth/Fog/M_RiftVolumeFog
graph {
 properties {blendMode=BLEND_Additive, materialDomain=MD_Volume, pixelDepthOffsetMode=PDOM_Legacy, usedShadingModels=Default Lit}
 nodes [
  n1 = Add(A=ObjectPositionWS(), B=VectorParameter() {parameterName=offset, defaultValue=<rgba 0 0 -1445.9475 1>}.RGB)
  n2 = WorldPosition()
  n3 = ScalarParameter() {defaultValue=2140.7742, parameterName=Radius}
  n4 = ScalarParameter() {defaultValue=0.8, parameterName=h3}
  n5 = Divide(A=WorldPosition().XYZ, B=ScalarParameter() {defaultValue=1024, parameterName=NoiseSize})
  n6 = ComponentMask(None=@n5) {g=true, r=true}
  n7 = Multiply(A=Time(), B=ScalarParameter() {defaultValue=0.15, parameterName=NoiseSpeed})
  n8 = ComponentMask(None=@n5) {b=true}
  n9 = TextureSampleParameter2D!(UVs=Add(A=Add(A=@n6, B=@n7), B=Divide(A=@n8) {constB=4}), Level, Apply View MipBias) {constMipValue=0, mipValueMode=TMVM_MipLevel, parameterName=NoiseTexture, texture=/Game/Effects/FogVolume/T_TileNoise}
  n10 = ComponentMask(None=@n5) {b=true, g=true}
  n11 = Multiply(A=@n7) {constB=0.5}
  n12 = ComponentMask(None=@n5) {r=true}
  n13 = TextureSampleParameter2D(UVs=Add(A=Add(A=@n10, B=@n11), B=Divide(A=@n12) {constB=4})) {constMipValue=0, mipValueMode=TMVM_MipLevel, parameterName=NoiseTexture, texture=&T_TileNoise}
  n14 = Clamp(None=LinearInterpolate(A=ScalarParameter() {parameterName=NoiseMin}, B=ScalarParameter() {defaultValue=1, parameterName=NoiseMax}, Alpha=Min(A=@n9.R, B=@n13.R)))
  n15 = ScalarParameter() {defaultValue=0.5, parameterName=NoNoiseOpacityValue}
  n16 = QualitySwitch(Default=@n14, Low=@n15, High=@n14, Medium=@n15)
  n17 = ScalarParameter() {defaultValue=2, parameterName=NoiseStrength}
  n18 = Clamp(None=Subtract(A=SphereMask(A=@n1, B=@n2.XYZ, Radius=@n3, Hardness=@n4) {hardnessPercent=50}, B=Multiply(A=@n16, B=@n17)))
  Multiply(A=VectorParameter() {parameterName=Color, defaultValue=<rgba 3 0.9531 2.7441 1>}.RGB, B=@n18)
  Multiply(
   A=VectorParameter() {parameterName=Emis, defaultValue=<rgba 0.2188 0 0.167 1>}.RGB
   B=Multiply(A=@n18, B=ScalarParameter() {defaultValue=0.1, parameterName=Ext})
  )
 ]
}
